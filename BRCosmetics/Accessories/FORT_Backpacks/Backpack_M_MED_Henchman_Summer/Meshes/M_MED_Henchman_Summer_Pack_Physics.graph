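# Physics asset for the Henchman Summer backpack cosmetic, covering a shovel and left and right tanks.
# Each part pairs a tiny sphere body on its base bone with a capsule body on its parent-named bone.
# Three constraints link each capsule to its own base with narrow swing and very tight twist limits, the shovel swinging slightly less than the tanks.
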PhysicsAsset /BRCosmetics/Accessories/FORT_Backpacks/Backpack_M_MED_Henchman_Summer/Meshes/M_MED_Henchman_Summer_Pack_Physics
bodies {
 dyn_shovel_base [{shapeType=Sphere, radius=0.5}]
 dyn_shovel_parent [{shapeType=Capsule, length=12.2775, radius=2.2305, center=<0 0 1.1167>, rotation=<rot 0 0 90>}]
 dyn_tank_base_l [{shapeType=Sphere, radius=0.5}]
 dyn_tank_base_r [{shapeType=Sphere, radius=0.5}]
 dyn_tank_parent_l [{shapeType=Capsule, length=4.369, radius=3.8219, center=<3.3914 -0.0002 0>}]
 dyn_tank_parent_r [{shapeType=Capsule, length=3.1185, radius=4.1936, center=<3.1751 0.0001 0>}]
}
constraints [
 {bone1Name=dyn_tank_parent_l, bone2Name=dyn_tank_base_l, swing1LimitDegrees=20, swing2LimitDegrees=20, twistLimitDegrees=3}
 {bone1Name=dyn_tank_parent_r, bone2Name=dyn_tank_base_r, swing1LimitDegrees=20, swing2LimitDegrees=20, twistLimitDegrees=3}
 {bone1Name=dyn_shovel_parent, bone2Name=dyn_shovel_base, swing1LimitDegrees=15, swing2LimitDegrees=15, twistLimitDegrees=3}
]
